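# Surface material for the deform body of the large male player base character, flagged for skeletal mesh and mesh deformer use.
# Multiplies an ambient occlusion texture by a Color parameter, giving a flat tint with baked shading rather than an albedo map.
# A normal map and Metal, Spec and Rough scalar parameters supply the remaining Default Lit surface inputs, with no masks or layering.
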Material /Game/Characters/Player/Male/Large/Bodies/M_LRG_Base/Materials/M_LRG_Base_Deform_Body
graph {
 properties {bUsedWithMeshDeformer=true, bUsedWithSkeletalMesh=true, pixelDepthOffsetMode=PDOM_Legacy, usedShadingModels=Default Lit}
 nodes [
  Multiply(
   A=TextureSample() {texture=/Game/Characters/Player/Male/Large/Bodies/M_LRG_Base/Textures/T_M_LRG_Base_Deform_AO}.RGB
   B=VectorParameter() {parameterName=Color, defaultValue=<rgba 0.1979 0.1806 0.1509 1>}.RGB
  )
  ScalarParameter() {parameterName=Metal}
  ScalarParameter() {defaultValue=0.5, parameterName=Spec}
  ScalarParameter() {defaultValue=0.3, parameterName=Rough}
  TextureSample() {samplerType=SAMPLERTYPE_Normal, texture=/Game/Characters/Player/Male/Large/Bodies/M_LRG_Base/Textures/T_M_LRG_Base_Deform_N}
 ]
}
